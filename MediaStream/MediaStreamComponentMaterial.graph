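# Translucent unlit material for the MediaStream component, built around a texture parameter for media.
# A texture object parameter named MediaTexture feeds a single texture sample that uses the shared clamp sampler.
# The parameter defaults to an engine editor Matinee group texture, which serves as a placeholder.
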Material /MediaStream/MediaStreamComponentMaterial
graph {
 properties {blendMode=BLEND_Translucent, shadingModel=MSM_Unlit, usedShadingModels=Unlit}
 nodes [TextureSample(Tex=TextureObjectParameter() {parameterName=MediaTexture, texture=/Engine/EditorMaterials/MatineeGroups/MAT_Groups_Anim}) {samplerSource=SSM_Clamp_WorldGroupSettings}]
}
